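# ActivtyCheckLine is a Save the World lobby ready-up widget that shows a single game activity check as a checkbox and label row.
# Built on FortActivtyCheckLine, its Configure event breaks the check struct, writing CheckText to the uppercase label and bPassedCheck to the checkbox's selected state.
# The scaled-down checkbox is hit-test invisible, so it only indicates pass or fail.
# The struct's Solution field is unused here.
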WidgetBlueprint /EcosystemReadyUpErrorsUI/Lobby/STW/ActivtyCheckLine
graphs [EventGraph {nodes [
   n1 = Event Configure!() -> (OutputDelegate, then, Check:"Game Activity Check Structure (by ref)") {is=title K2Node_Event}
   n2 = Break Game Activity Check!(GameActivityCheck:"Game Activity Check Structure (by ref)"=@n1.Check) -> (CheckText:Text, bPassedCheck:bool=~, Solution:ESTWReadyUpError Enum) {is=title K2Node_BreakStruct}
   > "SetText (Text)"(self=Get Text!() -> (Text:Common Text Object Reference), InText=@n2)
   > SetSelected!(
    execute
    self:Fort Checkbox Object Reference=Get FortCheckbox() -> (FortCheckbox:Fort Checkbox Object Reference)
    NewSelected:bool=@n2.bPassedCheck
   ) -> (then)
  ]}]
variables []
widgets {
 info {inheritedWidgetCount=0, parentClass=/Script/FortniteUI.FortActivtyCheckLine, rootWidgetClass=&HorizontalBox, widgetCount=4}
 widgets [
  HorizontalBox_36 {widgetClassPath=&HorizontalBox, widget=HorizontalBox_36}
  ScaleBox_0 {
   widgetClassPath=&ScaleBox
   parent=HorizontalBox_36
   widget=ScaleBox_0
   properties {stretch=UserSpecified, userSpecifiedScale=0.5233}
   slot {class=HorizontalBoxSlot}
  }
  FortCheckbox {
   widgetClassPath=/FortniteCoreUI/Frontend/FortCheckbox
   bIsVariable=true
   parent=ScaleBox_0
   widget=FortCheckbox
   properties {visibility=HitTestInvisible}
   slot {class=ScaleBoxSlot}
  }
  Text {
   widgetClassPath=&CommonTextBlock
   bIsVariable=true
   parent=HorizontalBox_36
   widget=Text
   properties {
    justification=Center
    text="<Check Line Entry>"
    textTransformPolicy=ToUpper
    font {
     bForceMonospaced=false
     bMaterialIsStencil=false
     fontMaterial=null
     fontObject=&BurbankSmall
     letterSpacing=0
     monospacedWidth=1
     size=22
     skewAmount=0
     typefaceFontName=Bold
     outlineSettings {
      bApplyOutlineToDropShadows=false
      bMiteredCorners=false
      bSeparateFillAlpha=false
      outlineMaterial=null
      outlineSize=0
      outlineColor=<rgba 0 0 0 1>
     }
    }
    shadowOffset=<0 0>
   }
   slot {class=HorizontalBoxSlot, properties {verticalAlignment=VAlign_Center, padding=<ltrb 12 0 0 0>}}
  }
 ]
}
