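# UI-domain material for a Llama background border style, drawing a soft radial gradient with an alternate dark variant.
# Two exponential radial gradients share an off-center focal point, with separately tunable radius and density.
# The base gradient runs from a blue Light center to a navy Shadow edge, while the alternate runs from XBlack to XRimLight.
# The BlackOn scalar crossfades between the two looks, defaulting to the blue one.
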Material /FortUILibrary/Styles/Border/Materials/M-Border-LlamaBackground
graph {
 properties {bUsedWithStaticMesh=false, materialDomain=MD_UI, pixelDepthOffsetMode=PDOM_Legacy, usedShadingModels=Default Lit}
 nodes [
  n1 = TextureCoordinate()
  n2 = Constant2Vector() {g=0.4, r=0.5}
  n3 = ScalarParameter() {defaultValue=0.75, parameterName=Radius, sliderMax=1}
  n4 = ScalarParameter() {defaultValue=1, parameterName=Density, sliderMax=1}
  n5 = TextureCoordinate()
  n6 = Constant2Vector() {g=0.4, r=0.5}
  n7 = ScalarParameter() {defaultValue=0.75, parameterName=XRadius, sliderMax=1}
  n8 = ScalarParameter() {defaultValue=1, parameterName=XDensity, sliderMax=1}
  LinearInterpolate(
   A=LinearInterpolate(
    A=VectorParameter() {parameterName=Shadow, defaultValue=<rgba 0.003 0.0073 0.0312 0>}.RGB
    B=VectorParameter() {parameterName=Light, defaultValue=<rgba 0.005 0.1666 0.5625 0>}.RGB
    Alpha=MaterialFunctionCall(UVs=@n1, CenterPosition=@n2, Radius=@n3, Density=@n4) {materialFunction=&RadialGradientExponential}.RadialGradientExponential
   )
   B=LinearInterpolate(
    A=VectorParameter() {parameterName=XRimLight, defaultValue=<rgba 0.0801 0.1611 0.2135 0>}.RGB
    B=VectorParameter() {parameterName=XBlack}.RGB
    Alpha=MaterialFunctionCall(UVs=@n5, CenterPosition=@n6, Radius=@n7, Density=@n8) {materialFunction=&RadialGradientExponential}.RadialGradientExponential
   )
   Alpha=ScalarParameter() {parameterName=BlackOn}
  )
 ]
}
