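# Save the World ready-up error dialog widget, built on FortCampaignReadyScreen, listing requirements the player and squad must meet to play.
# A blurred, tilted banner holds the title, message and an inherited LinesEntryBox of ActivtyCheckLine entries.
# ConfigureButtons uses the ESTWReadyUpError solution to switch between Install and Settings buttons, or collapse the switcher for None.
# Install, Settings and Close clicks call parent functions to install content, show settings or close the dialog.
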WidgetBlueprint /EcosystemReadyUpErrorsUI/Lobby/STW/CampaignReadyScreen
graphs [EventGraph {nodes [
   "On Button Clicked (Button_Close)"!() -> (OutputDelegate, then, Button:Common Button Legacy Object Reference) {is=title K2Node_ComponentBoundEvent}
   > CloseDialog(self:Fort Campaign Ready Screen Object Reference)
   "On Button Clicked (Button_Settings)"!() -> (OutputDelegate, then, Button:Common Button Legacy Object Reference) {is=title K2Node_ComponentBoundEvent}
   > ShowSettingsScreen(self:Fort Campaign Ready Screen Object Reference)
   "On Button Clicked (Button_Install)"!() -> (OutputDelegate, then, Button:Common Button Legacy Object Reference) {is=title K2Node_ComponentBoundEvent}
   > InstallSaveTheWorldContent(self:Fort Campaign Ready Screen Object Reference)
   n1 = Event ConfigureButtons!() -> (OutputDelegate, then, Solution:ESTWReadyUpError Enum=None) {is=title K2Node_Event}
   > n2 = Branch(Condition="Equal (Enum)"(A:ESTWReadyUpError Enum=@n1.Solution, B:ESTWReadyUpError Enum=None))
   n3 = Get OptionButtonsSwitcher() -> (OptionButtonsSwitcher:Widget Switcher Object Reference)
   > SetVisibility(self=@n3, InVisibility=Collapsed)
   SetVisibility(execute=@n2.else, self=@n3, InVisibility=SelfHitTestInvisible)
   > SetActiveWidget(
    self:Widget Switcher Object Reference=@n3
    Widget=Select!(
     None:Icon Text Button Object Reference
     Install:Icon Text Button Object Reference=Get Button_Install() -> (Button_Install:Icon Text Button Object Reference)
     Settings:Icon Text Button Object Reference=Get Button_Settings() -> (Button_Settings:Icon Text Button Object Reference)
     Index:ESTWReadyUpError Enum=@n1.Solution
    ) -> (ReturnValue:Icon Text Button Object Reference)
   )
  ]}]
variables []
widgets {
 info {inheritedWidgetCount=1, parentClass=/Script/FortniteUI.FortCampaignReadyScreen, rootWidgetClass=&BackgroundBlur, widgetCount=17}
 widgets [
  BackgroundBlur_0 {widgetClassPath=&BackgroundBlur, widget=BackgroundBlur_0, properties {blurStrength=15}}
  Overlay_0 {
   widgetClassPath=&Overlay
   parent=BackgroundBlur_0
   widget=Overlay_0
   properties {clipping=ClipToBoundsAlways}
   slot {class=BackgroundBlurSlot, properties {padding=<ltrb 0 0 0 0>}}
  }
  RotationRoot {
   widgetClassPath=&VerticalBox
   parent=Overlay_0
   widget=RotationRoot
   properties {renderTransform {angle=-1, scale=<1 1>, shear=<0 0>, translation=<0 0>}}
   slot {class=OverlaySlot, properties {horizontalAlignment=HAlign_Fill, verticalAlignment=VAlign_Center}}
  }
  BackgroundColor {
   widgetClassPath=&Border
   parent=RotationRoot
   widget=BackgroundColor
   properties {
    background {
     bIsDynamicallyLoaded=false
     drawAs=Image
     imageType=NoImage
     mirroring=NoMirror
     resourceName=None
     resourceObject=/FortUILibrary/Styles/Border/Materials/M-Border-LlamaBackground
     tiling=NoTile
     imageSize=<32 32>
     margin=<ltrb 0 0 0 0>
     outlineSettings {
      bUseBrushTransparency=false
      roundingType=HalfHeightRadius
      width=0
      color {colorUseRule=UseColor_Specified, specifiedColor=<rgba 0 0 0 0>}
      cornerRadii=<0 0 0 0>
     }
     tintColor {colorUseRule=UseColor_Specified, specifiedColor=<rgba 1 1 1 1>}
     uVRegion {bIsValid=false, max=<0 0>, min=<0 0>}
    }
    brushColor=<rgba 0.0507 0.0521 0.0461 1>
    padding=<ltrb 0 0 0 0>
   }
   slot {class=VerticalBoxSlot, properties {padding=<ltrb -40 0 -40 0>}}
  }
  VerticalBox_0 {
   widgetClassPath=&VerticalBox
   parent=BackgroundColor
   widget=VerticalBox_0
   slot {class=BorderSlot, properties {padding=<ltrb 0 0 0 0>}}
  }
  Image_Separator_Up {
   widgetClassPath=&Image
   parent=VerticalBox_0
   widget=Image_Separator_Up
   properties {
    brush {
     bIsDynamicallyLoaded=false
     drawAs=Box
     imageType=NoImage
     mirroring=NoMirror
     resourceName=None
     resourceObject=/Game/Athena/UI/ReplayBrowser/M_Ui_Skewed_Vfor9slices_Inst1
     tiling=NoTile
     imageSize=<32 32>
     margin=<ltrb 0.5 0.5 0.5 0.5>
     outlineSettings {
      bUseBrushTransparency=false
      roundingType=HalfHeightRadius
      width=0
      color {colorUseRule=UseColor_Specified, specifiedColor=<rgba 0 0 0 0>}
      cornerRadii=<0 0 0 0>
     }
     tintColor {colorUseRule=UseColor_Specified, specifiedColor=<rgba 1 1 1 1>}
     uVRegion {bIsValid=false, max=<0 0>, min=<0 0>}
    }
    colorAndOpacity=<rgba 0 0.0908 1 1>
    renderTransform {angle=0, scale=<1.2 1>, shear=<0 0>, translation=<0 -12>}
   }
   slot {class=VerticalBoxSlot, properties {verticalAlignment=VAlign_Top}}
  }
  UnrotateInner {
   widgetClassPath=&VerticalBox
   parent=VerticalBox_0
   widget=UnrotateInner
   properties {renderTransform {angle=1, scale=<1 1>, shear=<0 0>, translation=<0 0>}}
   slot {class=VerticalBoxSlot, properties {padding=<ltrb 0 40 0 40>}}
  }
  CommonText_Title {
   widgetClassPath=&CommonTextBlock
   bIsVariable=true
   parent=UnrotateInner
   widget=CommonText_Title
   properties {
    autoWrapText=true
    justification=Center
    lineHeightPercentage=0.9
    text=SAVE THE WORLD
    textTransformPolicy=ToUpper
    font {
     bForceMonospaced=false
     bMaterialIsStencil=false
     fontMaterial=null
     fontObject=&Burbank
     letterSpacing=0
     monospacedWidth=1
     size=52
     skewAmount=0
     typefaceFontName=Burbank-Black
     outlineSettings {
      bApplyOutlineToDropShadows=false
      bMiteredCorners=false
      bSeparateFillAlpha=false
      outlineMaterial=null
      outlineSize=0
      outlineColor=<rgba 0 0 0 1>
     }
    }
    margin=<ltrb 0 0 0 12>
    renderTransform {angle=0, scale=<1 1>, shear=<-12 0>, translation=<0 0>}
    shadowOffset=<0 0>
   }
   slot {class=VerticalBoxSlot}
  }
  CommonText_MainError {
   widgetClassPath=&CommonTextBlock
   bIsVariable=true
   parent=UnrotateInner
   widget=CommonText_MainError
   properties {
    autoWrapText=true
    justification=Center
    text="To PLAY SAVE THE WORLD, you AND your squad must meet the following requirements:"
    textTransformPolicy=ToUpper
    font {
     bForceMonospaced=false
     bMaterialIsStencil=false
     fontMaterial=null
     fontObject=&BurbankSmall
     letterSpacing=0
     monospacedWidth=1
     size=18
     skewAmount=0
     typefaceFontName=Bold
     outlineSettings {
      bApplyOutlineToDropShadows=false
      bMiteredCorners=false
      bSeparateFillAlpha=false
      outlineMaterial=null
      outlineSize=0
      outlineColor=<rgba 0 0 0 1>
     }
    }
    shadowOffset=<0 0>
   }
   slot {class=VerticalBoxSlot, properties {padding=<ltrb 12 0 12 0>}}
  }
  SizeBox_0 {
   widgetClassPath=&SizeBox
   parent=UnrotateInner
   widget=SizeBox_0
   properties {bOverride_MinDesiredWidth=true, minDesiredWidth=1060}
   slot {class=VerticalBoxSlot, properties {horizontalAlignment=HAlign_Center, padding=<ltrb 0 0 0 20>}}
  }
  LinesEntryBox {
   widgetClassPath=&DynamicEntryBox
   bInherited=true
   bIsVariable=true
   parent=SizeBox_0
   widget=LinesEntryBox
   properties {entryBoxType=Vertical, entryWidgetClass=&/ActivtyCheckLine.ActivtyCheckLine_C}
   slot {class=SizeBoxSlot, properties {padding=<ltrb 0 30 0 0>}}
  }
  Image_Separator_Down {
   widgetClassPath=&Image
   parent=VerticalBox_0
   widget=Image_Separator_Down
   properties {
    brush {
     bIsDynamicallyLoaded=false
     drawAs=Box
     imageType=NoImage
     mirroring=NoMirror
     resourceName=None
     resourceObject=&M_Ui_Skewed_Vfor9slices_Inst1
     tiling=NoTile
     imageSize=<32 32>
     margin=<ltrb 0.5 0.5 0.5 0.5>
     outlineSettings {
      bUseBrushTransparency=false
      roundingType=HalfHeightRadius
      width=0
      color {colorUseRule=UseColor_Specified, specifiedColor=<rgba 0 0 0 0>}
      cornerRadii=<0 0 0 0>
     }
     tintColor {colorUseRule=UseColor_Specified, specifiedColor=<rgba 1 1 1 1>}
     uVRegion {bIsValid=false, max=<0 0>, min=<0 0>}
    }
    colorAndOpacity=<rgba 0 0.0908 1 1>
    renderTransform {angle=0, scale=<1.2 1>, shear=<0 0>, translation=<0 12>}
   }
   slot {class=VerticalBoxSlot, properties {verticalAlignment=VAlign_Top}}
  }
  HorizontalBox_0 {
   widgetClassPath=&HorizontalBox
   parent=RotationRoot
   widget=HorizontalBox_0
   properties {renderTransform {angle=0, scale=<1 1>, shear=<0 0>, translation=<0 -32>}}
   slot {class=VerticalBoxSlot, properties {horizontalAlignment=HAlign_Center, verticalAlignment=VAlign_Bottom}}
  }
  OptionButtonsSwitcher {
   widgetClassPath=&WidgetSwitcher
   bIsVariable=true
   parent=HorizontalBox_0
   widget=OptionButtonsSwitcher
   slot {class=HorizontalBoxSlot, properties {horizontalAlignment=HAlign_Center, verticalAlignment=VAlign_Center, padding=<ltrb 12 0 12 0>}}
  }
  Button_Install {
   widgetClassPath=&IconTextButton
   bIsVariable=true
   parent=OptionButtonsSwitcher
   widget=Button_Install
   properties {
    bDisplayInputActionWhenNotInteractable=false
    bIgnoreInputActionWidgetText=true
    bIsFocusable=false
    buttonText=INSTALL
    minHeight=55
    minWidth=256
    onButtonBaseClicked=(&self:WidgetTree.Button_Install.HandleOnButtonClicked)
    onButtonBaseDoubleClicked=(&self:WidgetTree.Button_Install.HandleOnButtonDoubleClicked)
    onButtonBaseHovered=(&self:WidgetTree.Button_Install.HandleOnButtonHovered)
    onButtonBaseUnhovered=(&self:WidgetTree.Button_Install.HandleOnButtonUnhovered)
    onSelectedChangedBase=(&self:WidgetTree.Button_Install.HandleOnSelectedChanged)
    style=/FortUILibrary/Foundation/Buttons/ButtonStyles/Primary/ButtonStyle-Primary-M_Skew_Yellow.ButtonStyle-Primary-M_Skew_Yellow_C
    triggeringInputAction {dataTable=&InputActionDataTable, rowName=Confirm}
   }
   slot {class=WidgetSwitcherSlot}
  }
  Button_Settings {
   widgetClassPath=&IconTextButton
   bIsVariable=true
   parent=OptionButtonsSwitcher
   widget=Button_Settings
   properties {
    bDisplayInputActionWhenNotInteractable=false
    bIgnoreInputActionWidgetText=true
    bIsFocusable=false
    buttonText=SETTINGS
    minHeight=55
    minWidth=256
    onButtonBaseClicked=(&self:WidgetTree.Button_Settings.HandleOnButtonClicked)
    onButtonBaseDoubleClicked=(&self:WidgetTree.Button_Settings.HandleOnButtonDoubleClicked)
    onButtonBaseHovered=(&self:WidgetTree.Button_Settings.HandleOnButtonHovered)
    onButtonBaseUnhovered=(&self:WidgetTree.Button_Settings.HandleOnButtonUnhovered)
    onSelectedChangedBase=(&self:WidgetTree.Button_Settings.HandleOnSelectedChanged)
    style=&ButtonStyle-Primary-M_Skew_Yellow_C
    triggeringInputAction {dataTable=&InputActionDataTable, rowName=Confirm}
   }
   slot {class=WidgetSwitcherSlot}
  }
  Button_Close {
   widgetClassPath=&IconTextButton
   bIsVariable=true
   parent=HorizontalBox_0
   widget=Button_Close
   properties {
    bIgnoreInputActionWidgetText=true
    bIsFocusable=false
    buttonText=Close
    minHeight=55
    minWidth=256
    onButtonBaseClicked=(&self:WidgetTree.Button_Close.HandleOnButtonClicked)
    onButtonBaseDoubleClicked=(&self:WidgetTree.Button_Close.HandleOnButtonDoubleClicked)
    onButtonBaseHovered=(&self:WidgetTree.Button_Close.HandleOnButtonHovered)
    onButtonBaseUnhovered=(&self:WidgetTree.Button_Close.HandleOnButtonUnhovered)
    onSelectedChangedBase=(&self:WidgetTree.Button_Close.HandleOnSelectedChanged)
    style=/FortUILibrary/Foundation/Buttons/ButtonStyles/Primary/ButtonStyle-Primary-M_Skew_Blue.ButtonStyle-Primary-M_Skew_Blue_C
    triggeringInputAction {dataTable=&InputActionDataTable, rowName=Decline}
   }
   slot {class=HorizontalBoxSlot, properties {horizontalAlignment=HAlign_Center, verticalAlignment=VAlign_Center, padding=<ltrb 12 0 12 0>}}
  }
 ]
}
